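# Niagara swing trail for the PearlTaffy MicStand pickaxe, made of two anim-trail ribbons plus a burst of symbol sprites.
# Trail_Color and Trail_Accent spawn through FN anim-trail modules, dimmed by a front-end preview flag and time of day, fading over short lifetimes.
# Symbols spawns per distance moved, partitioning particles eight ways to pick color and sub-image, then pops in size and drifts with inherited velocity and drag.
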NiagaraSystem /SparksCosmetics/Instrument/MicStand/PearlTaffy/Effects/NS_Pickaxe_MicStand_PearlTaffy_SwingTrail
dependencies {
 usedDataInterfaces [
  &NiagaraDataInterfaceCurve
  &NiagaraDataInterfaceArrayFloat
  &NiagaraDataInterfaceArrayFloat2
  &NiagaraDataInterfaceVector2DCurve
  &NiagaraDataInterfaceColorCurve
 ]
 usedDynamicInputs [
  &SelectVector2DFromArray
  &ReturnNormalizedExecIndex
  &RandomRangeVector
 ]
 usedModules [
  &SystemState
  &self:SystemSpawnScript.SystemScriptSource.SystemScriptGraph.NiagaraNodeAssignment_0.SetVariables_#1
  &self:SystemSpawnScript.SystemScriptSource.SystemScriptGraph.NiagaraNodeAssignment_1.SetVariables_#2
 ]
}
emitters {
 Symbols {
  emitterUpdateScript [
   &EmitterState()
   &SpawnRate(SpawnRate=20, Use Spawn Probability=false) {enabled=false}
   &SpawnPerUnit(Module.Use Max Movement Threshold=false, Module.Use Movement Tolerance=false, Spawn Spacing=35, Use Spawn Probability=false)
  ]
  particleSpawnScript [
   &PartitionParticles(Number Of Partitions=8, Partition Mode:ENiagara_PartitionMode=::NewEnumerator5)
   &InitializeParticle(
    Color=&MergeHelper_ColorSwitch(
     0=<rgba 1 1 1 1>
     1=<rgba 1 1 1 1>
     2=<rgba 0.1787 0.0591 2 1>
     3=<rgba 0 2 1.3368 1>
     4=<rgba 2 0.1074 0.1523 1>
     5=<rgba 1 1 1 1>
     6=<rgba 2 0.1 1.3667 1>
     7=<rgba 1 1 1 1>
     Index=$Particles.PartitionParticles.Partition
     Num=::NewEnumerator6
    )
    Lifetime Max=0.4
    Lifetime Min=0.1
    Lifetime Mode=::NewEnumerator1
    Sprite Rotation Mode=::NewEnumerator2
    Sprite Size Mode=::NewEnumerator3
    Uniform Sprite Size=&FloatFromCurve(FloatCurve=DataInterface() {propertyValues {Curve {keys [<at 0 4 RCIM_Cubic>, {arriveTangent=3.2538, tangentMode=RCTM_User, time=1, value=8}]}}})
    UsePositionOffset=true
    Write Material Random=true
   )
   &ShapeLocation(
    MaskSpawnGroup=false
    Offset=<0 0 10>
    Offset Coordinate Space=::Local
    Offset Mode=::NewEnumerator0
    Override Seed=false
    Rotation Angle Type:ENiagara_AngleInput=::NewEnumerator0
    Rotation Axis=<0 1 0>
    Rotation Mode=::NewEnumerator1
    Sphere Radius=15
    Sphere Surface Distribution=1
    UseOwnerScale=true
   )
   &AddVelocity(
    Module.Use Falloff Distance=false
    Override Seed=false
    Velocity Mode=::NewEnumerator1
    Velocity Speed=&RandomRangeFloat(Minimum=30, Maximum=50)
   )
   &InheritVelocity(Inherited Velocity Amount Scale=<0.2 0.2 0.2>, Speed Limit=true, UseSpeedThreshold=false)
   &CameraOffset(Camera Offset Amount=10)
   &AddVelocity(Velocity=<200 0 0>) {enabled=false, module=AddVelocity001}
  ]
  particleUpdateScript [
   &ParticleState()
   &ScaleSpriteSize(
    Non-Uniform Curve Sprite Scale=DataInterface() {propertyValues {
      XCurve {
       defaultValue=0
       keys [
        {arriveTangent=27.0378, interpMode=RCIM_Cubic, leaveTangent=27.0378, tangentMode=RCTM_User, time=0, value=0}
        {
         arriveTangent=-6.4665
         arriveTangentWeight=0.0469
         interpMode=RCIM_Cubic
         leaveTangent=-6.4665
         tangentMode=RCTM_User
         time=0.0289
         value=1.195
        }
        {arriveTangent=-0.7599, interpMode=RCIM_Cubic, leaveTangent=-0.7599, time=0.0957, value=0.5946}
        {arriveTangent=1.0763, interpMode=RCIM_Cubic, leaveTangent=1.0763, time=0.2498, value=1.0271}
        {arriveTangent=-0.484, interpMode=RCIM_Cubic, leaveTangent=-0.484, time=0.5511, value=1.0847}
        {arriveTangent=-2.4165, interpMode=RCIM_Cubic, leaveTangent=-2.4165, time=0.8845, value=0.72}
        {arriveTangent=-0.0405, arriveTangentWeight=0.2864, interpMode=RCIM_Cubic, leaveTangent=-0.0405, tangentMode=RCTM_User, time=1, value=0}
       ]
      }
      YCurve {
       defaultValue=0
       keys [
        {arriveTangent=6.4321, interpMode=RCIM_Cubic, leaveTangent=6.4321, tangentMode=RCTM_User, time=0, value=0}
        {arriveTangent=10.7376, interpMode=RCIM_Cubic, leaveTangent=10.7376, time=0.0648, value=0.9991}
        {arriveTangent=-0.5061, interpMode=RCIM_Cubic, leaveTangent=-0.5061, tangentMode=RCTM_User, time=0.1399, value=1.5024}
        {arriveTangent=-1.4131, interpMode=RCIM_Cubic, leaveTangent=-1.4131, tangentMode=RCTM_User, time=0.6265, value=0.8211}
        {arriveTangent=-4.8374, interpMode=RCIM_Cubic, leaveTangent=-4.8374, tangentMode=RCTM_User, time=0.839, value=0.869}
        {arriveTangent=-0.0405, interpMode=RCIM_Cubic, leaveTangent=-0.0405, tangentMode=RCTM_User, time=1, value=0}
       ]
      }
     }}
    Scale Sprite Size Mode=::NewEnumerator3
   )
   /SetVariables(Particles.SubImageIndex=&Multiply_FloatByInt(Float=2, Integer=$Particles.PartitionParticles.Partition)) {isSetParameters=true}
   &DynamicMaterialParameters(
    Index 0 Param 1=&FloatFromCurve(FloatCurve=DataInterface() {propertyValues {Curve {keys [
         {arriveTangent=-0.1162, interpMode=RCIM_Cubic, leaveTangent=-0.1162, tangentMode=RCTM_User, time=0.2, value=1}
         {arriveTangent=-0.0508, interpMode=RCIM_Cubic, leaveTangent=-0.0508, tangentMode=RCTM_User, time=0.5, value=1}
         {tangentMode=RCTM_User, time=1, value=0}
        ]}}})
    Index 0 Param 2=&MergeHelper_FloatSwitch(
     0=&SelectFloatFromArray(
      Array Sampling Mode=::NewEnumerator0
      Direct Array Index=?
      Float Selection Array=DataInterface() {propertyValues {FloatData [0,0]}}
      Random Seed=0
     )
     1=0
     2=0
     3=0
     4=0
     5=&RandomRangeFloat(Minimum=0, Maximum=1)
     6=0
     7=&RandomRangeFloat(Minimum=0, Maximum=1)
     Index=$Particles.PartitionParticles.Partition
     Num=::NewEnumerator6
    )
    Module.Index0Param0Write=true
    Module.Index0Param1Write=true
    Module.Index0Param2Write=false
    Module.Index0Param3Write=false
   )
   &Drag(Drag=3, Use Linear Drag=true, Use Rotational Drag=true)
   &SolveForcesAndVelocity(Clamp Velocity=false, Limit Acceleration=false)
  ]
  rendererClasses [&NiagaraSpriteRendererProperties]
  renderers [{rendererClass=&NiagaraSpriteRendererProperties}]
 }
 Trail_Accent {
  emitterUpdateScript [
   &EmitterState()
   &FN_SpawnAnimTrail_v2()
  ]
  particleSpawnScript [
   /Game/Effects/Niagara/Modules/AnimTrail/FN_InitializeAnimTrail/FN_InitializeAnimTrailBoneIndecies(First Bone or Socket Index=1)
   /Game/Effects/Niagara/Modules/AnimTrail/FN_InitializeAnimTrail/FN_InitializeAnimTrail(
    AnimTrailWidth=1
    Color=&MakeCustomLinearColorFromBool(Bool=$User.bIsFrontEndPreview, True Color=<rgba 1 1 1 0.5>, False Color=<rgba 1 1 1 1>)
    Lifetime=0.15
   )
   &AddVelocity(Velocity=<500 0 0>) {enabled=false}
  ]
  particleUpdateScript [
   &FN_UpdateAnimTrail_v2()
   &ScaleColor(
    Scale Alpha=&Multiply_Float(
     A=&FloatFromCurve(FloatCurve=DataInterface() {propertyValues {Curve {keys [
          {interpMode=RCIM_Cubic, leaveTangent=-3.4731, tangentMode=RCTM_Break, time=0, value=1}
          <at 0.6689 0.0233>
          {arriveTangent=-0.2139, interpMode=RCIM_Cubic, tangentMode=RCTM_Break, time=1, value=0}
         ]}}})
     B=&FNTimeOfDayScalar(Night Scale=0.65, Morning Scale=0.5, Day Scale=0.7, Evening Scale=0.7)
    )
    ScaleA=true
    ScaleRGB=true
   )
   &SolveForcesAndVelocity(Clamp Velocity=false, Limit Acceleration=false)
   &DynamicMaterialParameters(
    Index 0 Param 1=1
    Index 0 Param 4=1
    Index 1 Param 3=&FloatFromCurve(
     FloatCurve=DataInterface() {propertyValues {Curve {keys [<at 0 0>, {arriveTangent=1.9814, tangentMode=RCTM_User, time=1, value=1}]}}}
     CurveIndex=&ModuloFloat(A=$Emitter.Age, B=1)
     Scale Curve=0
    )
    Index 1 Param 4=1
    Index 2 Param 1=10
    Index 2 Param 2=1
    Index 2 Param 3=0.5
    Index 2 Param 4=0.6
    Module.Index0Float0Write=true
    Module.Index0Float1Write=true
    Module.Index0Float2Write=true
    Module.Index0Float3Write=true
    Module.Index1Float0Write=true
    Module.Index1Float1Write=true
    Module.Index1Float2Write=true
    Module.Index1Float3Write=true
    Module.Index2Float0Write=true
    Module.Index2Float1Write=true
    Module.Index2Float2Write=true
    Module.Index2Float3Write=true
    Write Parameter Index 1=::NewEnumerator1
    Write Parameter Index 2=::NewEnumerator1
   )
  ]
  rendererClasses [&NiagaraRibbonRendererProperties]
  renderers [{rendererClass=&NiagaraRibbonRendererProperties}]
 }
 Trail_Color {
  emitterUpdateScript [
   &EmitterState()
   &FN_SpawnAnimTrail_v2()
  ]
  particleSpawnScript [
   &FN_InitializeAnimTrailBoneIndecies(First Bone or Socket Index=1)
   &FN_InitializeAnimTrail(
    AnimTrailWidth=1
    Color=&MakeCustomLinearColorFromBool(Bool=$User.bIsFrontEndPreview, True Color=<rgba 1 1 1 0.5>, False Color=<rgba 1 1 1 1>)
    Lifetime=0.2
   )
   &AddVelocity(Velocity=<500 0 0>) {enabled=false}
  ]
  particleUpdateScript [
   &ScaleColor(
    Scale Alpha=&FloatFromCurve(
     FloatCurve=DataInterface() {propertyValues {Curve {keys [
         {interpMode=RCIM_Cubic, leaveTangent=-4.0875, tangentMode=RCTM_Break, time=0, value=1}
         {arriveTangent=-0.4222, interpMode=RCIM_Cubic, leaveTangent=-0.4222, tangentMode=RCTM_User, time=0.5009, value=0.1221}
         <at 0.6689 0.0233>
         {arriveTangent=-0.2139, interpMode=RCIM_Cubic, tangentMode=RCTM_Break, time=1, value=0}
        ]}}}
     Scale Curve=&FNTimeOfDayScalar(Night Scale=0.7, Morning Scale=0.9, Day Scale=0.75, Evening Scale=1)
    )
    Scale RGB=&FN_Select_Vector_from_Time_Of_Day(Morning=<0.9 0.9 0.9>, Day=<1 1 1>, Evening=<1 1 1>, Night=<0.5 0.5 1>)
    ScaleA=true
    ScaleRGB=true
   ) {module=ScaleColor002}
   &FN_UpdateAnimTrail_v2()
   &SolveForcesAndVelocity(Clamp Velocity=false, Limit Acceleration=false)
   &DynamicMaterialParameters(
    Index 0 Param 1=1
    Index 1 Param 3=&FloatFromCurve(
     FloatCurve=DataInterface() {propertyValues {Curve {keys [<at 0 0>, {arriveTangent=1.9814, tangentMode=RCTM_User, time=1, value=1}]}}}
     CurveIndex=&ModuloFloat(A=$Emitter.Age, B=1)
     Scale Curve=0
    )
    Index 1 Param 4=1
    Index 2 Param 1=10
    Index 2 Param 2=1
    Index 2 Param 3=0.5
    Index 2 Param 4=0.6
    Module.Index0Float0Write=true
    Module.Index0Float1Write=true
    Module.Index0Float2Write=true
    Module.Index0Float3Write=true
    Module.Index1Float0Write=true
    Module.Index1Float1Write=true
    Module.Index1Float2Write=true
    Module.Index1Float3Write=true
    Module.Index2Float0Write=true
    Module.Index2Float1Write=true
    Module.Index2Float2Write=true
    Module.Index2Float3Write=true
    Write Parameter Index 1=::NewEnumerator1
    Write Parameter Index 2=::NewEnumerator1
   )
  ]
  rendererClasses [&NiagaraRibbonRendererProperties]
  renderers [{rendererClass=&NiagaraRibbonRendererProperties}]
 }
}
summary {
 systemName=NS_Pickaxe_MicStand_PearlTaffy_SwingTrail
 emitters [
  {emitterName=Trail_Color, rendererClasses [&NiagaraRibbonRendererProperties]}
  {emitterName=Trail_Accent, rendererClasses [&NiagaraRibbonRendererProperties]}
  {emitterName=Symbols, rendererClasses [&NiagaraSpriteRendererProperties]}
 ]
 userVariables [
  User.bIsFrontEndPreview {type=NiagaraBool, defaultValue=true}
  User.RibbonIDOffset {type=NiagaraInt32}
  User.SkeletalMesh {
   type=NiagaraDataInterfaceSkeletalMesh
   defaultValue {
    struct=NiagaraExt_VariableValue_DataInterface
    value {dataInterface=&self:NiagaraDataInterfaceSkeletalMesh_0}
   }
  }
 ]
}
